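# Sound cue for the shield destruction in the Habitat NPC content, playing the orb shield break wave.
# A short random delay feeds a modulator that varies pitch and volume slightly on the single wave player.
# It uses the Habitat boss attenuation and gadgets sound class, with a concurrency override that keeps simultaneous instances to a small count.
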SoundCue /NPCLibrary/Audio/Habitat/Habitat_Shield_Destroy_Cue
root=Delay_0
nodes [
 {id=Delay_0, children [Modulator_1], properties {delayMax=0.3, delayMin=0.2}}
 {id=Modulator_1, children [WavePlayer_1], properties {pitchMax=1.1, pitchMin=0.97, volumeMax=1.3, volumeMin=1}}
 {id=WavePlayer_1, properties {soundWaveAssetPtr=&/Habitat_Orb_Shield_Break}}
]
properties {
 attenuationSettings=&/Habitat_Boss_Att
 bOverrideConcurrency=true
 maxDistance=4900
 soundClassObject=&Fort_Gadgets
 volumeMultiplier=0.4286
 concurrencyOverrides {maxCount=2}
}
