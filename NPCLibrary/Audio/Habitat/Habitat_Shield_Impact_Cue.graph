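# Sound cue for impacts on the Habitat shield, played as a gadget-class sound with the Habitat boss attenuation.
# A random node picks equally between two shield impact waves, then a modulator slightly varies pitch and a very short random delay offsets playback.
# A concurrency override caps it at two simultaneous instances, and the cue plays at a reduced volume.
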SoundCue /NPCLibrary/Audio/Habitat/Habitat_Shield_Impact_Cue
root=Delay_0
nodes [
 {id=Delay_0, children [Modulator_0], properties {delayMax=0.02, delayMin=0.01}}
 {id=Modulator_0, children [Random_0], properties {pitchMax=1.02, pitchMin=0.97, volumeMin=1}}
 {id=Random_0, children [WavePlayer_1,WavePlayer_2], properties {weights [1,1]}}
 {id=WavePlayer_1, properties {soundWaveAssetPtr=&/Habitat_Shield_Impact}}
 {id=WavePlayer_2, properties {soundWaveAssetPtr=&/Habitat_Shield_Impact_v2}}
]
properties {
 attenuationSettings=&/Habitat_Boss_Att
 bOverrideConcurrency=true
 maxDistance=4900
 soundClassObject=&Fort_Gadgets
 volumeMultiplier=0.3571
 concurrencyOverrides {maxCount=2}
}
