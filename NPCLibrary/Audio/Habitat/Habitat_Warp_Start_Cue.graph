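# Sound cue for the start of a Habitat warp in the NPCLibrary audio, using Habitat boss attenuation and the Fort_Gadgets sound class.
# A single wave player feeds a modulator that slightly randomizes pitch and volume on each play, under a reduced cue volume.
# Concurrency override is on with a small voice cap, so it takes precedence over the referenced Evil_Boss_Con set from the TinyBalloon folder.
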SoundCue /NPCLibrary/Audio/Habitat/Habitat_Warp_Start_Cue
root=Modulator_1
nodes [
 {id=Modulator_1, children [WavePlayer_0], properties {pitchMax=1.1, pitchMin=0.98, volumeMax=1.3, volumeMin=1}}
 {id=WavePlayer_0, properties {soundWaveAssetPtr=&/Habitat_Warp_Start_V1}}
]
properties {
 attenuationSettings=&/Habitat_Boss_Att
 bOverrideConcurrency=true
 maxDistance=4900
 soundClassObject=&Fort_Gadgets
 volumeMultiplier=0.5
 concurrencyOverrides {maxCount=2}
 concurrencySet [/NPCLibrary/Audio/TinyBalloon/Evil_Boss_Con]
}
