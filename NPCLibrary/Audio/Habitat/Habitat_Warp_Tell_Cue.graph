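# Sound cue for the Habitat warp tell in the NPC library, using the Habitat boss attenuation.
# A single wave player for Habitat_Warp_Tell_V1 feeds a modulator that randomizes pitch and volume on each play.
# The cue routes to the Fort_Gadgets sound class and borrows the ThunderMeow concurrency rules to limit overlapping instances.
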SoundCue /NPCLibrary/Audio/Habitat/Habitat_Warp_Tell_Cue
root=Modulator_1
nodes [
 {id=Modulator_1, children [WavePlayer_2], properties {pitchMin=0.98, volumeMax=1.3, volumeMin=1}}
 {id=WavePlayer_2, properties {soundWaveAssetPtr=&/Habitat_Warp_Tell_V1}}
]
properties {
 attenuationSettings=&/Habitat_Boss_Att
 maxDistance=4900
 priority=0.5
 soundClassObject=&Fort_Gadgets
 volumeMultiplier=0.5357
 concurrencySet [/NPCLibrary/Audio/ThunderMeow/SFX/ThunderMeow_Concurrency]
}
