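# Sound cue for the wolf despawn (derez) effect in the NPC Habitat audio set.
# A Random node picks one of three derez wave variants with equal weight, feeding a Modulator that slightly varies pitch and volume.
# It uses boss attenuation and the gadgets sound class at reduced volume.
# Concurrency references the Habitat summons set, with single-instance PreventNew override values.
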SoundCue /NPCLibrary/Audio/Habitat/Habitat_Wolves_Despawn_Cue
root=Modulator_0
nodes [
 {id=Modulator_0, children [Random_0], properties {pitchMax=1.03, pitchMin=0.99, volumeMax=1.1}}
 {id=Random_0, children [WavePlayer_0,WavePlayer_1,WavePlayer_2], properties {weights [1,1,1]}}
 {id=WavePlayer_0, properties {soundWaveAssetPtr=&/s22_Habitat_Wolf_Derez_v1}}
 {id=WavePlayer_1, properties {soundWaveAssetPtr=&/s22_Habitat_Wolf_Derez_v2}}
 {id=WavePlayer_2, properties {soundWaveAssetPtr=&/s22_Habitat_Wolf_Derez_v3}}
]
properties {
 attenuationSettings=&/Habitat_Boss_Att
 maxDistance=4900
 priority=0.5
 soundClassObject=&Fort_Gadgets
 volumeMultiplier=0.5
 concurrencyOverrides {maxCount=1, resolutionRule=PreventNew}
 concurrencySet [&/Habitat_Summons_Conc]
}
